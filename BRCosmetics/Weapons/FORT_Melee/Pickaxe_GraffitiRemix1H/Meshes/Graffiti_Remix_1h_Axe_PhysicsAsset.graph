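# Physics asset for the Graffiti Remix one-handed axe pickaxe, giving its handle and roller collision shapes and the roller a jointed spin.
# It has a capsule on the handle bone, a sideways capsule on dyn_roller and a small box on dyn_roller_MSTR.
# A single constraint joins dyn_roller to dyn_roller_MSTR with both swings locked and twist free, so the roller rotates about one axis.
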PhysicsAsset /BRCosmetics/Weapons/FORT_Melee/Pickaxe_GraffitiRemix1H/Meshes/Graffiti_Remix_1h_Axe_PhysicsAsset
bodies {
 Melee_TwoHanded_Handle [{shapeType=Capsule, length=57.757, radius=4.9384, center=<0 0 40.6654>}]
 dyn_roller [{shapeType=Capsule, length=18, radius=6, center=<0 -15 0>, rotation=<rot 0 0 90>}]
 dyn_roller_MSTR [{shapeType=Box, extentX=7.5, extentY=7.5, extentZ=7.5}]
}
constraints [{
  bone1Name=dyn_roller
  bone2Name=dyn_roller_MSTR
  swing1LimitDegrees=0
  swing1Motion=Locked
  swing2LimitDegrees=45
  swing2Motion=Locked
  twistLimitDegrees=45
  twistMotion=Free
 }]
